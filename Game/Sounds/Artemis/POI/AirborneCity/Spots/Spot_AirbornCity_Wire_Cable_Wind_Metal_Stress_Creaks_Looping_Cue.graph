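# Ambient spot SoundCue for Airborne City cables, playing wind-driven metal stress and creaks with shared wire-stress attenuation.
# The first two quality levels mix a wind loop, randomly present or absent, with randomly delayed, repeating layered pairs of groans and bridge creaks.
# Nested looping envelopers shape the wind with 30-second pitch and 60-second volume cycles.
# The other two levels play only one random knock, without looping.
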SoundCue /Game/Sounds/Artemis/POI/AirborneCity/Spots/Spot_AirbornCity_Wire_Cable_Wind_Metal_Stress_Creaks_Looping_Cue
root=QualityLevel_0
nodes [
 {
  id=Attenuation_1
  children [Looping_0]
  properties {attenuationSettings=/Game/Sounds/Artemis/POI/Blimps_Crashsites/Spots/Spot_Metal_Wire_Stress_Capsule_Liner_Att}
 }
 {
  id=Attenuation_2
  children [Random_2]
  properties {attenuationSettings=&Spot_Metal_Wire_Stress_Capsule_Liner_Att}
 }
 {
  id=Attenuation_3
  children [Random_3]
  properties {attenuationSettings=&Spot_Metal_Wire_Stress_Capsule_Liner_Att}
 }
 {id=Delay_0, children [Modulator_0], properties {delayMax=8, delayMin=2}}
 {
  id=Enveloper_0
  children [WavePlayer_23]
  properties {
   bLoop=true
   bLoopIndefinitely=true
   loopEnd=30
   pitchCurve {editorCurveData {keys [{interpMode=RCIM_Cubic, time=0}, <at 15.9468 0.9181 RCIM_Cubic>, {interpMode=RCIM_Cubic, time=30}]}}
   volumeCurve {editorCurveData {keys [<at 0 1>]}}
  }
 }
 {
  id=Enveloper_3
  children [Enveloper_0]
  properties {
   bLoop=true
   bLoopIndefinitely=true
   loopEnd=60
   pitchCurve {editorCurveData {keys [{time=0}]}}
   volumeCurve {editorCurveData {keys [
      {arriveTangent=-0.01, interpMode=RCIM_Cubic, time=0, value=1}
      {arriveTangent=-0.0537, interpMode=RCIM_Cubic, leaveTangent=-0.0537, time=10, value=1.2}
      {arriveTangent=0.001, interpMode=RCIM_Cubic, leaveTangent=0.001, tangentMode=RCTM_User, time=14.8985, value=0.2}
      {arriveTangent=-0.0072, interpMode=RCIM_Cubic, leaveTangent=-0.0072, time=20, value=1}
      {arriveTangent=-0.0286, interpMode=RCIM_Cubic, leaveTangent=-0.0286, tangentMode=RCTM_User, time=25.6275, value=0.123}
      {arriveTangent=0.1042, interpMode=RCIM_Cubic, leaveTangent=0.1042, time=30.2698, value=0.1}
      {arriveTangent=0.0271, interpMode=RCIM_Cubic, leaveTangent=0.0271, time=35, value=1.1}
      <at 40 0.364>
      {arriveTangent=0.0318, interpMode=RCIM_Cubic, leaveTangent=0.0318, time=52.0586, value=0.3704}
      <at 60 1>
     ]}}
  }
 }
 {id=Looping_0, children [Delay_0]}
 {id=Mixer_0, children [Attenuation_3, Attenuation_1], properties {inputVolume [0.5,1]}}
 {id=Mixer_1, children [Random_0,Random_1], properties {inputVolume [1,1]}}
 {id=Modulator_0, children [Mixer_1], properties {volumeMax=1.2}}
 {id=QualityLevel_0, children [Mixer_0, Mixer_0, Attenuation_2, Attenuation_2]}
 {
  id=Random_0
  children [
   WavePlayer_1
   WavePlayer_2
   WavePlayer_3
   WavePlayer_4
   WavePlayer_5
   WavePlayer_13
   WavePlayer_14
   WavePlayer_15
   WavePlayer_16
   WavePlayer_0
   WavePlayer_6
   WavePlayer_7
  ]
  properties {weights [1,1,1,1,1,1,1,1,1,1,1,1]}
 }
 {
  id=Random_1
  children [WavePlayer_8, WavePlayer_9, WavePlayer_10, WavePlayer_11, WavePlayer_12, WavePlayer_17, WavePlayer_18, WavePlayer_19]
  properties {weights [1,1,1,1,1,1,1,1]}
 }
 {id=Random_2, children [WavePlayer_20, WavePlayer_21, WavePlayer_22], properties {weights [1,1,1]}}
 {id=Random_3, children [Enveloper_3,null], properties {weights [1,1]}}
 {id=WavePlayer_0, properties {soundWaveAssetPtr=/Game/Sounds/Apollo/Biomes/Generic/AMB_Metal_Stress_Wronk_Groan_Knock_09}}
 {id=WavePlayer_1, properties {soundWaveAssetPtr=/Game/Sounds/Apollo/Biomes/Generic/AMB_Metal_Stress_Wronk_Groan_Knock_01}}
 {id=WavePlayer_10, properties {soundWaveAssetPtr=/Game/Sounds/Artemis/POI/Big_Bridge/Spots/Spot_Big_Bridge_Metal_Creak_02}}
 {id=WavePlayer_11, properties {soundWaveAssetPtr=/Game/Sounds/Artemis/POI/Big_Bridge/Spots/Spot_Big_Bridge_Metal_Creak_03}}
 {id=WavePlayer_12, properties {soundWaveAssetPtr=/Game/Sounds/Artemis/POI/Big_Bridge/Spots/Spot_Big_Bridge_Metal_Creak_04}}
 {id=WavePlayer_13, properties {soundWaveAssetPtr=/Game/Sounds/Apollo/Biomes/Generic/AMB_Metal_Stress_Wronk_Groan_Knock_02}}
 {id=WavePlayer_14, properties {soundWaveAssetPtr=/Game/Sounds/Apollo/Biomes/Generic/AMB_Metal_Stress_Wronk_Groan_Stretch_02}}
 {id=WavePlayer_15, properties {soundWaveAssetPtr=/Game/Sounds/Apollo/Biomes/Generic/AMB_Metal_Stress_Wronk_Groan_Stretch_03}}
 {id=WavePlayer_16, properties {soundWaveAssetPtr=/Game/Sounds/Apollo/Biomes/Generic/AMB_Metal_Stress_Wronk_Groan_Stretch_04}}
 {id=WavePlayer_17, properties {soundWaveAssetPtr=/Game/Sounds/Artemis/POI/Big_Bridge/Spots/Spot_Big_Bridge_Metal_Creak_05}}
 {id=WavePlayer_18, properties {soundWaveAssetPtr=/Game/Sounds/Artemis/POI/Big_Bridge/Spots/Spot_Big_Bridge_Metal_Creak_Large_01}}
 {id=WavePlayer_19, properties {soundWaveAssetPtr=/Game/Sounds/Artemis/POI/Big_Bridge/Spots/Spot_Big_Bridge_Metal_Creak_Large_02}}
 {id=WavePlayer_2, properties {soundWaveAssetPtr=/Game/Sounds/Apollo/Biomes/Generic/AMB_Metal_Stress_Wronk_Groan_Knock_06}}
 {id=WavePlayer_20, properties {soundWaveAssetPtr=&AMB_Metal_Stress_Wronk_Groan_Knock_01}}
 {id=WavePlayer_21, properties {soundWaveAssetPtr=&AMB_Metal_Stress_Wronk_Groan_Knock_06}}
 {id=WavePlayer_22, properties {soundWaveAssetPtr=/Game/Sounds/Apollo/Biomes/Generic/AMB_Metal_Stress_Wronk_Groan_Knock_05}}
 {
  id=WavePlayer_23
  properties {bLooping=true, soundWaveAssetPtr=&/Spot_AirborneCity_Wind_Excited_Metal_Loop_02}
 }
 {id=WavePlayer_3, properties {soundWaveAssetPtr=&AMB_Metal_Stress_Wronk_Groan_Knock_05}}
 {id=WavePlayer_4, properties {soundWaveAssetPtr=/Game/Sounds/Apollo/Biomes/Generic/AMB_Metal_Stress_Wronk_Groan_Knock_04}}
 {id=WavePlayer_5, properties {soundWaveAssetPtr=/Game/Sounds/Apollo/Biomes/Generic/AMB_Metal_Stress_Wronk_Groan_Knock_03}}
 {id=WavePlayer_6, properties {soundWaveAssetPtr=/Game/Sounds/Apollo/Biomes/Generic/AMB_Metal_Stress_Wronk_Groan_Knock_07}}
 {id=WavePlayer_7, properties {soundWaveAssetPtr=/Game/Sounds/Apollo/Biomes/Generic/AMB_Metal_Stress_Wronk_Groan_Knock_08}}
 {id=WavePlayer_8, properties {soundWaveAssetPtr=/Game/Sounds/Artemis/POI/Big_Bridge/Spots/Spot_Big_Bridge_Metal_Creak_Large_03}}
 {id=WavePlayer_9, properties {soundWaveAssetPtr=/Game/Sounds/Artemis/POI/Big_Bridge/Spots/Spot_Big_Bridge_Metal_Creak_01}}
]
properties {
 maxDistance=2140.1951
 soundClassObject=&Fort_Ambient_Outdoors_Athena
 volumeMultiplier=0.902
 preEffectBusSends [{}]
}
