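# Sound cue for big metal stress-groan spot ambience at the AirborneCity point of interest, reusing the Big_Bridge groan recordings.
# A Random node picks among seven equally weighted groan waves, then a Modulator slightly varies the pitch of each play.
# It plays in the outdoor ambient sound class at reduced volume, using a positional one-shot tile-ambient attenuation.
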SoundCue /Game/Sounds/Artemis/POI/AirborneCity/Spots/Spot_AirborneCity_Big_Metal_Stress_Groans_Cue
root=Modulator_0
nodes [
 {id=Modulator_0, children [Random_0], properties {pitchMax=1.1, pitchMin=0.9}}
 {
  id=Random_0
  children [WavePlayer_0,WavePlayer_1,WavePlayer_2,WavePlayer_3,WavePlayer_4,WavePlayer_5,WavePlayer_6]
  properties {weights [1,1,1,1,1,1,1]}
 }
 {id=WavePlayer_0, properties {soundWaveAssetPtr=/Game/Sounds/Artemis/POI/Big_Bridge/Spots/Spot_Big_Metal_Bridge_Huge_Stress_Groan_01}}
 {id=WavePlayer_1, properties {soundWaveAssetPtr=/Game/Sounds/Artemis/POI/Big_Bridge/Spots/Spot_Big_Metal_Bridge_Huge_Stress_Groan_02}}
 {id=WavePlayer_2, properties {soundWaveAssetPtr=/Game/Sounds/Artemis/POI/Big_Bridge/Spots/Spot_Big_Metal_Bridge_Huge_Stress_Groan_03}}
 {id=WavePlayer_3, properties {soundWaveAssetPtr=/Game/Sounds/Artemis/POI/Big_Bridge/Spots/Spot_Big_Metal_Bridge_Huge_Stress_Groan_04}}
 {id=WavePlayer_4, properties {soundWaveAssetPtr=/Game/Sounds/Artemis/POI/Big_Bridge/Spots/Spot_Big_Metal_Bridge_Huge_Stress_Groan_05}}
 {id=WavePlayer_5, properties {soundWaveAssetPtr=/Game/Sounds/Artemis/POI/Big_Bridge/Spots/Spot_Big_Metal_Bridge_Huge_Stress_Groan_06}}
 {id=WavePlayer_6, properties {soundWaveAssetPtr=/Game/Sounds/Artemis/POI/Big_Bridge/Spots/Spot_Big_Metal_Bridge_Huge_Stress_Groan_07}}
]
properties {
 attenuationSettings=/Game/Sounds/Attenuation/Fort_TileAmbient_OneShot_MorePositional_Att
 maxDistance=8000
 soundClassObject=&Fort_Ambient_Outdoors_Athena
 volumeMultiplier=0.551
 preEffectBusSends [{}]
}
