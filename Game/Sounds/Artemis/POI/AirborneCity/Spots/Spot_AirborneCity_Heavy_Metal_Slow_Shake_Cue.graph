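# Sound cue for an Airborne City POI ambient one-shot of heavy metal slowly shaking, reusing Blimp crash site recordings.
# A Random node picks one of six equally weighted waves and passes it through a Modulator for variation.
# The root Quality Level node feeds that chain to its first two inputs and leaves the other two empty.
# Playback uses the outdoor ambient sound class and tile one-shot attenuation, at reduced volume.
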SoundCue /Game/Sounds/Artemis/POI/AirborneCity/Spots/Spot_AirborneCity_Heavy_Metal_Slow_Shake_Cue
root=QualityLevel_0
nodes [
 {id=Modulator_0, children [Random_2]}
 {id=QualityLevel_0, children [Modulator_0,Modulator_0,null,null]}
 {
  id=Random_2
  children [WavePlayer_20, WavePlayer_21, WavePlayer_22, WavePlayer_23, WavePlayer_24, WavePlayer_25]
  properties {weights [1,1,1,1,1,1]}
 }
 {
  id=WavePlayer_20
  properties {soundWaveAssetPtr=/Game/Sounds/Artemis/POI/Blimps_Crashsites/Spots/Spot_Blimp_Crashsite_Heavy_Metal_Slow_Shake_01}
 }
 {
  id=WavePlayer_21
  properties {soundWaveAssetPtr=/Game/Sounds/Artemis/POI/Blimps_Crashsites/Spots/Spot_Blimp_Crashsite_Heavy_Metal_Slow_Shake_02}
 }
 {
  id=WavePlayer_22
  properties {soundWaveAssetPtr=/Game/Sounds/Artemis/POI/Blimps_Crashsites/Spots/Spot_Blimp_Crashsite_Heavy_Metal_Slow_Shake_03}
 }
 {
  id=WavePlayer_23
  properties {soundWaveAssetPtr=/Game/Sounds/Artemis/POI/Blimps_Crashsites/Spots/Spot_Blimp_Crashsite_Heavy_Metal_Slow_Shake_04}
 }
 {
  id=WavePlayer_24
  properties {soundWaveAssetPtr=/Game/Sounds/Artemis/POI/Blimps_Crashsites/Spots/Spot_Blimp_Crashsite_Heavy_Metal_Slow_Shake_05}
 }
 {
  id=WavePlayer_25
  properties {soundWaveAssetPtr=/Game/Sounds/Artemis/POI/Blimps_Crashsites/Spots/Spot_Blimp_Crashsite_Heavy_Metal_Slow_Shake_06}
 }
]
properties {
 attenuationSettings=/Game/Sounds/Attenuation/Fort_TileAmbient_OneShot_Att
 maxDistance=8000
 soundClassObject=&Fort_Ambient_Outdoors_Athena
 volumeMultiplier=0.501
 preEffectBusSends [{}]
}
