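# Sound cue for an outdoor ambient spot one-shot of singing metal resonances at the Airborne City POI, reusing Blimp Crashsite recordings.
# An equal-weight Random node picks one of eight waves, the first boosted through a Mixer.
# The same Random feeds the first two Quality Level slots, leaving the other two empty, and a root Modulator varies pitch.
# Playback uses tile-ambient one-shot attenuation and the outdoor ambient sound class.
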SoundCue /Game/Sounds/Artemis/POI/AirborneCity/Spots/Spot_AirborneCity_Metal_Singing_Resonsances_Cue
root=Modulator_0
nodes [
 {id=Mixer_0, children [WavePlayer_20, null], properties {inputVolume [1.2,1]}}
 {id=Modulator_0, children [QualityLevel_0], properties {pitchMax=1.1, pitchMin=0.8}}
 {id=QualityLevel_0, children [Random_2,Random_2,null,null]}
 {
  id=Random_2
  children [Mixer_0, WavePlayer_21, WavePlayer_22, WavePlayer_23, WavePlayer_24, WavePlayer_25, WavePlayer_26, WavePlayer_27]
  properties {weights [1,1,1,1,1,1,1,1]}
 }
 {
  id=WavePlayer_20
  properties {soundWaveAssetPtr=/Game/Sounds/Artemis/POI/Blimps_Crashsites/Spots/Spot_Blimp_Crashsite_Metal_Singing_Resonances_01}
 }
 {
  id=WavePlayer_21
  properties {soundWaveAssetPtr=/Game/Sounds/Artemis/POI/Blimps_Crashsites/Spots/Spot_Blimp_Crashsite_Metal_Singing_Resonances_02}
 }
 {
  id=WavePlayer_22
  properties {soundWaveAssetPtr=/Game/Sounds/Artemis/POI/Blimps_Crashsites/Spots/Spot_Blimp_Crashsite_Metal_Singing_Resonances_03}
 }
 {
  id=WavePlayer_23
  properties {soundWaveAssetPtr=/Game/Sounds/Artemis/POI/Blimps_Crashsites/Spots/Spot_Blimp_Crashsite_Metal_Singing_Resonances_04}
 }
 {
  id=WavePlayer_24
  properties {soundWaveAssetPtr=/Game/Sounds/Artemis/POI/Blimps_Crashsites/Spots/Spot_Blimp_Crashsite_Metal_Singing_Resonances_05}
 }
 {
  id=WavePlayer_25
  properties {soundWaveAssetPtr=/Game/Sounds/Artemis/POI/Blimps_Crashsites/Spots/Spot_Blimp_Crashsite_Metal_Singing_Resonances_06}
 }
 {
  id=WavePlayer_26
  properties {soundWaveAssetPtr=/Game/Sounds/Artemis/POI/Blimps_Crashsites/Spots/Spot_Blimp_Crashsite_Metal_Singing_Resonances_07}
 }
 {
  id=WavePlayer_27
  properties {soundWaveAssetPtr=/Game/Sounds/Artemis/POI/Blimps_Crashsites/Spots/Spot_Blimp_Crashsite_Metal_Singing_Resonances_08}
 }
]
properties {
 attenuationSettings=/Game/Sounds/Attenuation/Fort_TileAmbient_OneShot_Att
 maxDistance=8000
 soundClassObject=&Fort_Ambient_Outdoors_Athena
 volumeMultiplier=0.451
 preEffectBusSends [{}]
}
