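# Ambient spot cue for the Airborne City POI that plays quiet, intermittent light metal-stress sounds, named for small interior spaces.
# A loop repeats a random delay into a quality switch with only its first two slots connected.
# Those slots share a pitch/volume modulator over ten equally weighted, reused Tilted Towers waves.
# Overrides set short-range logarithmic falloff with listener focus and reverb send, plus a low concurrency cap.
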SoundCue /Game/Sounds/Artemis/POI/AirborneCity/Spots/Spot_AirborneCity_Metal_Stress_Light_Loop_Small_Sphere_Int_Cue
root=Looping_0
nodes [
 {id=Delay_0, children [QualityLevel_0], properties {delayMax=16, delayMin=3}}
 {id=Looping_0, children [Delay_0]}
 {id=Modulator_0, children [Random_0], properties {pitchMin=0.9}}
 {id=QualityLevel_0, children [Modulator_0,Modulator_0,null,null]}
 {
  id=Random_0
  children [
   WavePlayer_3
   WavePlayer_4
   WavePlayer_5
   WavePlayer_6
   WavePlayer_7
   WavePlayer_8
   WavePlayer_9
   WavePlayer_10
   WavePlayer_11
   WavePlayer_12
  ]
  properties {weights [1,1,1,1,1,1,1,1,1,1]}
 }
 {id=WavePlayer_10, properties {soundWaveAssetPtr=/Game/Sounds/Artemis/POI/Tilted_Towers/Spots/Spot_Tilted_Towers_Metal_Stress_Light_08}}
 {id=WavePlayer_11, properties {soundWaveAssetPtr=/Game/Sounds/Artemis/POI/Tilted_Towers/Spots/Spot_Tilted_Towers_Metal_Stress_Light_09}}
 {id=WavePlayer_12, properties {soundWaveAssetPtr=/Game/Sounds/Artemis/POI/Tilted_Towers/Spots/Spot_Tilted_Towers_Metal_Stress_Light_10}}
 {id=WavePlayer_3, properties {soundWaveAssetPtr=/Game/Sounds/Artemis/POI/Tilted_Towers/Spots/Spot_Tilted_Towers_Metal_Stress_Light_01}}
 {id=WavePlayer_4, properties {soundWaveAssetPtr=/Game/Sounds/Artemis/POI/Tilted_Towers/Spots/Spot_Tilted_Towers_Metal_Stress_Light_02}}
 {id=WavePlayer_5, properties {soundWaveAssetPtr=/Game/Sounds/Artemis/POI/Tilted_Towers/Spots/Spot_Tilted_Towers_Metal_Stress_Light_03}}
 {id=WavePlayer_6, properties {soundWaveAssetPtr=/Game/Sounds/Artemis/POI/Tilted_Towers/Spots/Spot_Tilted_Towers_Metal_Stress_Light_04}}
 {id=WavePlayer_7, properties {soundWaveAssetPtr=/Game/Sounds/Artemis/POI/Tilted_Towers/Spots/Spot_Tilted_Towers_Metal_Stress_Light_05}}
 {id=WavePlayer_8, properties {soundWaveAssetPtr=/Game/Sounds/Artemis/POI/Tilted_Towers/Spots/Spot_Tilted_Towers_Metal_Stress_Light_06}}
 {id=WavePlayer_9, properties {soundWaveAssetPtr=/Game/Sounds/Artemis/POI/Tilted_Towers/Spots/Spot_Tilted_Towers_Metal_Stress_Light_07}}
]
properties {
 attenuationSettings=/Game/Sounds/Attenuation/Fort_TileAmbient_OneShot_LessPositional_Log_Att
 bOverrideAttenuation=true
 bOverrideConcurrency=true
 maxDistance=1250
 soundClassObject=&Fort_Ambient_OneShots
 volumeMultiplier=0.25
 attenuationOverrides {
  bEnableFocusInterpolation=true
  bEnableListenerFocus=true
  bEnableSubmixSends=true
  distanceAlgorithm=Logarithmic
  falloffDistance=1250
  nonFocusVolumeAttenuation=0.4
  attenuationShapeExtents {x=0}
  submixSendSettings [{
    bEnableHPFCutoff=false
    bEnableLPFCutoff=false
    disableManualSendClamp=false
    hPFCutoff=20
    lPFCutoff=20000
    maxSendDistance=1000
    maxSendLevel=1
    minSendDistance=250
    minSendLevel=0.2
    sendLevel=1
    sendLevelControlMethod=Linear
    soundSubmix=&Fortnite_Reverb_Default
    customSendLevelCurve {editorCurveData {keys [<at 0 0.4>,<at 1 1>]}}
   }]
 }
 concurrencyOverrides {maxCount=2}
 soundSubmixSends [{minSendLevel=1, soundSubmix=&Fortnite_Reverb_Default}]
}
