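# Sound cue for an outdoor ambient spot at the Airborne City POI, looping randomized creaks of stressed metal trusses.
# A looping node repeats a 6–8 second random delay, then mixes one random generic metal groan with one random Big Bridge creak.
# A quality-level node feeds that mix to its first two slots only, leaving the other two empty.
# Overrides add logarithmic, listener-focus attenuation and limit playback to one voice.
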SoundCue /Game/Sounds/Artemis/POI/AirborneCity/Spots/Spot_AirborneCity_Metal_Truss_Stress_Creaks_Rnd_Loop_Ext_Cue
root=Looping_0
nodes [
 {id=Delay_0, children [QualityLevel_0], properties {delayMax=8, delayMin=6}}
 {id=Looping_0, children [Delay_0]}
 {id=Mixer_1, children [Random_0,Random_1], properties {inputVolume [1,1]}}
 {id=QualityLevel_0, children [Mixer_1,Mixer_1,null,null]}
 {
  id=Random_0
  children [
   WavePlayer_1
   WavePlayer_2
   WavePlayer_3
   WavePlayer_4
   WavePlayer_5
   WavePlayer_13
   WavePlayer_14
   WavePlayer_15
   WavePlayer_16
   WavePlayer_0
   WavePlayer_6
   WavePlayer_7
  ]
  properties {weights [1,1,1,1,1,1,1,1,1,1,1,1]}
 }
 {
  id=Random_1
  children [WavePlayer_8, WavePlayer_9, WavePlayer_10, WavePlayer_11, WavePlayer_12, WavePlayer_17, WavePlayer_18, WavePlayer_19]
  properties {weights [1,1,1,1,1,1,1,1]}
 }
 {id=WavePlayer_0, properties {soundWaveAssetPtr=/Game/Sounds/Apollo/Biomes/Generic/AMB_Metal_Stress_Wronk_Groan_Knock_09}}
 {id=WavePlayer_1, properties {soundWaveAssetPtr=/Game/Sounds/Apollo/Biomes/Generic/AMB_Metal_Stress_Wronk_Groan_Knock_01}}
 {id=WavePlayer_10, properties {soundWaveAssetPtr=/Game/Sounds/Artemis/POI/Big_Bridge/Spots/Spot_Big_Bridge_Metal_Creak_02}}
 {id=WavePlayer_11, properties {soundWaveAssetPtr=/Game/Sounds/Artemis/POI/Big_Bridge/Spots/Spot_Big_Bridge_Metal_Creak_03}}
 {id=WavePlayer_12, properties {soundWaveAssetPtr=/Game/Sounds/Artemis/POI/Big_Bridge/Spots/Spot_Big_Bridge_Metal_Creak_04}}
 {id=WavePlayer_13, properties {soundWaveAssetPtr=/Game/Sounds/Apollo/Biomes/Generic/AMB_Metal_Stress_Wronk_Groan_Knock_02}}
 {id=WavePlayer_14, properties {soundWaveAssetPtr=/Game/Sounds/Apollo/Biomes/Generic/AMB_Metal_Stress_Wronk_Groan_Stretch_02}}
 {id=WavePlayer_15, properties {soundWaveAssetPtr=/Game/Sounds/Apollo/Biomes/Generic/AMB_Metal_Stress_Wronk_Groan_Stretch_03}}
 {id=WavePlayer_16, properties {soundWaveAssetPtr=/Game/Sounds/Apollo/Biomes/Generic/AMB_Metal_Stress_Wronk_Groan_Stretch_04}}
 {id=WavePlayer_17, properties {soundWaveAssetPtr=/Game/Sounds/Artemis/POI/Big_Bridge/Spots/Spot_Big_Bridge_Metal_Creak_05}}
 {id=WavePlayer_18, properties {soundWaveAssetPtr=/Game/Sounds/Artemis/POI/Big_Bridge/Spots/Spot_Big_Bridge_Metal_Creak_Large_01}}
 {id=WavePlayer_19, properties {soundWaveAssetPtr=/Game/Sounds/Artemis/POI/Big_Bridge/Spots/Spot_Big_Bridge_Metal_Creak_Large_02}}
 {id=WavePlayer_2, properties {soundWaveAssetPtr=/Game/Sounds/Apollo/Biomes/Generic/AMB_Metal_Stress_Wronk_Groan_Knock_06}}
 {id=WavePlayer_3, properties {soundWaveAssetPtr=/Game/Sounds/Apollo/Biomes/Generic/AMB_Metal_Stress_Wronk_Groan_Knock_05}}
 {id=WavePlayer_4, properties {soundWaveAssetPtr=/Game/Sounds/Apollo/Biomes/Generic/AMB_Metal_Stress_Wronk_Groan_Knock_04}}
 {id=WavePlayer_5, properties {soundWaveAssetPtr=/Game/Sounds/Apollo/Biomes/Generic/AMB_Metal_Stress_Wronk_Groan_Knock_03}}
 {id=WavePlayer_6, properties {soundWaveAssetPtr=/Game/Sounds/Apollo/Biomes/Generic/AMB_Metal_Stress_Wronk_Groan_Knock_07}}
 {id=WavePlayer_7, properties {soundWaveAssetPtr=/Game/Sounds/Apollo/Biomes/Generic/AMB_Metal_Stress_Wronk_Groan_Knock_08}}
 {id=WavePlayer_8, properties {soundWaveAssetPtr=/Game/Sounds/Artemis/POI/Big_Bridge/Spots/Spot_Big_Bridge_Metal_Creak_Large_03}}
 {id=WavePlayer_9, properties {soundWaveAssetPtr=/Game/Sounds/Artemis/POI/Big_Bridge/Spots/Spot_Big_Bridge_Metal_Creak_01}}
]
properties {
 attenuationSettings=/Game/Sounds/Attenuation/Fort_TileAmbient_OneShot_LessPositional_Log_Att
 bOverrideAttenuation=true
 bOverrideConcurrency=true
 maxDistance=4350
 soundClassObject=&Fort_Ambient_OneShots_Outdoors
 volumeMultiplier=1
 attenuationOverrides {bEnableListenerFocus=true, distanceAlgorithm=Logarithmic, nonFocusVolumeAttenuation=0.75, attenuationShapeExtents {x=750}}
 concurrencyOverrides {maxCount=1, voiceStealReleaseTime=2}
 preEffectBusSends [{}]
}
