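# Looping ambient spot cue of metal stress sounds for the Airborne City underside, using cone-named attenuation and a concurrency override.
# Each loop waits a random delay, then a quality switch selects content.
# The first two slots randomly pick knocks, bridge creaks, or a half-weighted layer of stretch groans and huge swells.
# The last two slots play only UFO crash-site wronks; modulators vary volume or pitch throughout.
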SoundCue /Game/Sounds/Artemis/POI/AirborneCity/Spots/Spot_AirborneCity_Metal_Underside_Cone_Focussed_Lighter_Cue
root=Looping_0
nodes [
 {id=Delay_0, children [QualityLevel_0], properties {delayMax=5, delayMin=2}}
 {id=Looping_0, children [Delay_0]}
 {id=Mixer_2, children [Modulator_4,Modulator_2], properties {inputVolume [1,1]}}
 {id=Modulator_0, children [Random_2], properties {pitchMax=1.1, pitchMin=0.9, volumeMax=2.5, volumeMin=2}}
 {id=Modulator_2, children [Random_5], properties {pitchMin=0.8}}
 {id=Modulator_3, children [Random_0], properties {pitchMax=1, pitchMin=0.9, volumeMax=0.8, volumeMin=0.7}}
 {id=Modulator_4, children [Random_3], properties {pitchMin=0.9, volumeMax=0.8, volumeMin=0.6}}
 {id=Modulator_6, children [Random_8], properties {volumeMax=2.5, volumeMin=2}}
 {id=QualityLevel_0, children [Random_4,Random_4,Modulator_0,Modulator_0]}
 {
  id=Random_0
  children [WavePlayer_1, WavePlayer_7, WavePlayer_0, WavePlayer_4, WavePlayer_3, WavePlayer_2, WavePlayer_20, WavePlayer_21, WavePlayer_17]
  properties {weights [1,1,1,1,1,1,1,1,1]}
 }
 {id=Random_2, children [WavePlayer_25, WavePlayer_26, WavePlayer_27, WavePlayer_28, WavePlayer_29], properties {weights [1,1,1,1,1]}}
 {id=Random_3, children [WavePlayer_22, WavePlayer_23, WavePlayer_24], properties {weights [1,1,1]}}
 {id=Random_4, children [Modulator_3,Mixer_2,Modulator_6], properties {bRandomizeWithoutReplacement=false, weights [1,0.5,1]}}
 {id=Random_5, children [WavePlayer_8, WavePlayer_9, WavePlayer_10, WavePlayer_30, WavePlayer_31], properties {weights [1,1,1,1,1]}}
 {
  id=Random_8
  children [WavePlayer_42, WavePlayer_43, WavePlayer_44, WavePlayer_45, WavePlayer_46, WavePlayer_47, WavePlayer_48, WavePlayer_49]
  properties {weights [1,1,1,1,1,1,1,1]}
 }
 {id=WavePlayer_0, properties {soundWaveAssetPtr=/Game/Sounds/Apollo/Biomes/Generic/AMB_Metal_Stress_Wronk_Groan_Knock_03}}
 {id=WavePlayer_1, properties {soundWaveAssetPtr=/Game/Sounds/Apollo/Biomes/Generic/AMB_Metal_Stress_Wronk_Groan_Knock_01}}
 {
  id=WavePlayer_10
  properties {soundWaveAssetPtr=/Game/Sounds/Apollo/Biomes/UFO_Crash_Sites/AmbientOneshots/AMB_UFO_Crashsite_Huge_Metal_Swell_03}
 }
 {id=WavePlayer_17, properties {soundWaveAssetPtr=/Game/Sounds/Apollo/Biomes/Generic/AMB_Metal_Stress_Wronk_Groan_Knock_09}}
 {id=WavePlayer_2, properties {soundWaveAssetPtr=/Game/Sounds/Apollo/Biomes/Generic/AMB_Metal_Stress_Wronk_Groan_Knock_06}}
 {id=WavePlayer_20, properties {soundWaveAssetPtr=/Game/Sounds/Apollo/Biomes/Generic/AMB_Metal_Stress_Wronk_Groan_Knock_07}}
 {id=WavePlayer_21, properties {soundWaveAssetPtr=/Game/Sounds/Apollo/Biomes/Generic/AMB_Metal_Stress_Wronk_Groan_Knock_08}}
 {id=WavePlayer_22, properties {soundWaveAssetPtr=/Game/Sounds/Apollo/Biomes/Generic/AMB_Metal_Stress_Wronk_Groan_Stretch_02}}
 {id=WavePlayer_23, properties {soundWaveAssetPtr=/Game/Sounds/Apollo/Biomes/Generic/AMB_Metal_Stress_Wronk_Groan_Stretch_03}}
 {id=WavePlayer_24, properties {soundWaveAssetPtr=/Game/Sounds/Apollo/Biomes/Generic/AMB_Metal_Stress_Wronk_Groan_Stretch_04}}
 {
  id=WavePlayer_25
  properties {soundWaveAssetPtr=/Game/Sounds/Apollo/Biomes/UFO_Crash_Sites/AmbientOneshots/AMB_UFO_Crashsite_Metal_Wronk_Spot_05}
 }
 {
  id=WavePlayer_26
  properties {soundWaveAssetPtr=/Game/Sounds/Apollo/Biomes/UFO_Crash_Sites/AmbientOneshots/AMB_UFO_Crashsite_Metal_Wronk_Spot_08}
 }
 {
  id=WavePlayer_27
  properties {soundWaveAssetPtr=/Game/Sounds/Apollo/Biomes/UFO_Crash_Sites/AmbientOneshots/AMB_UFO_Crashsite_Metal_Wronk_Spot_09}
 }
 {
  id=WavePlayer_28
  properties {soundWaveAssetPtr=/Game/Sounds/Apollo/Biomes/UFO_Crash_Sites/AmbientOneshots/AMB_UFO_Crashsite_Metal_Wronk_Spot_02}
 }
 {
  id=WavePlayer_29
  properties {soundWaveAssetPtr=/Game/Sounds/Apollo/Biomes/UFO_Crash_Sites/AmbientOneshots/AMB_UFO_Crashsite_Metal_Wronk_Spot_10}
 }
 {id=WavePlayer_3, properties {soundWaveAssetPtr=/Game/Sounds/Apollo/Biomes/Generic/AMB_Metal_Stress_Wronk_Groan_Knock_05}}
 {
  id=WavePlayer_30
  properties {soundWaveAssetPtr=/Game/Sounds/Apollo/Biomes/UFO_Crash_Sites/AmbientOneshots/AMB_UFO_Crashsite_Huge_Metal_Swell_04}
 }
 {
  id=WavePlayer_31
  properties {soundWaveAssetPtr=/Game/Sounds/Apollo/Biomes/UFO_Crash_Sites/AmbientOneshots/AMB_UFO_Crashsite_Huge_Metal_Swell_05}
 }
 {id=WavePlayer_4, properties {soundWaveAssetPtr=/Game/Sounds/Apollo/Biomes/Generic/AMB_Metal_Stress_Wronk_Groan_Knock_04}}
 {id=WavePlayer_42, properties {soundWaveAssetPtr=/Game/Sounds/Artemis/POI/Big_Bridge/Spots/Spot_Big_Bridge_Metal_Creak_Large_03}}
 {id=WavePlayer_43, properties {soundWaveAssetPtr=/Game/Sounds/Artemis/POI/Big_Bridge/Spots/Spot_Big_Bridge_Metal_Creak_01}}
 {id=WavePlayer_44, properties {soundWaveAssetPtr=/Game/Sounds/Artemis/POI/Big_Bridge/Spots/Spot_Big_Bridge_Metal_Creak_02}}
 {id=WavePlayer_45, properties {soundWaveAssetPtr=/Game/Sounds/Artemis/POI/Big_Bridge/Spots/Spot_Big_Bridge_Metal_Creak_03}}
 {id=WavePlayer_46, properties {soundWaveAssetPtr=/Game/Sounds/Artemis/POI/Big_Bridge/Spots/Spot_Big_Bridge_Metal_Creak_04}}
 {id=WavePlayer_47, properties {soundWaveAssetPtr=/Game/Sounds/Artemis/POI/Big_Bridge/Spots/Spot_Big_Bridge_Metal_Creak_05}}
 {id=WavePlayer_48, properties {soundWaveAssetPtr=/Game/Sounds/Artemis/POI/Big_Bridge/Spots/Spot_Big_Bridge_Metal_Creak_Large_01}}
 {id=WavePlayer_49, properties {soundWaveAssetPtr=/Game/Sounds/Artemis/POI/Big_Bridge/Spots/Spot_Big_Bridge_Metal_Creak_Large_02}}
 {id=WavePlayer_7, properties {soundWaveAssetPtr=/Game/Sounds/Apollo/Biomes/Generic/AMB_Metal_Stress_Wronk_Groan_Knock_02}}
 {
  id=WavePlayer_8
  properties {soundWaveAssetPtr=/Game/Sounds/Apollo/Biomes/UFO_Crash_Sites/AmbientOneshots/AMB_UFO_Crashsite_Huge_Metal_Swell_01}
 }
 {
  id=WavePlayer_9
  properties {soundWaveAssetPtr=/Game/Sounds/Apollo/Biomes/UFO_Crash_Sites/AmbientOneshots/AMB_UFO_Crashsite_Huge_Metal_Swell_02}
 }
]
properties {
 attenuationSettings=&/Spot_AirbourneCity_Metal_Underside_No_Occ_Cone_Smaller_Log_Att
 bOverrideConcurrency=true
 maxDistance=9000
 soundClassObject=&Fort_Ambient_Outdoors_Athena
 volumeMultiplier=0.501
 concurrencyOverrides {maxCount=2, retriggerTime=3, voiceStealReleaseTime=5}
 preEffectBusSends [{}]
}
