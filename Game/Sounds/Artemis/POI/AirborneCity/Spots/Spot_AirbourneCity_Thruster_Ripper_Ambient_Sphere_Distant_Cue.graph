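# Sound cue for a looping, distant thruster-ripper ambient spot at the Airborne City POI, reusing a HoverPlatform wave and attenuation.
# A random start delay leads into an envelope fade-in, then pitch modulation, then a quality switch that feeds the looping wave.
# Only the first two quality slots are wired; the other two are empty.
# Its concurrency override caps voices at two with distance-based volume scaling, under the outdoor ambient sound class.
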SoundCue /Game/Sounds/Artemis/POI/AirborneCity/Spots/Spot_AirbourneCity_Thruster_Ripper_Ambient_Sphere_Distant_Cue
root=Delay_0
nodes [
 {id=Delay_0, children [Enveloper_0], properties {delayMax=2}}
 {
  id=Enveloper_0
  children [Modulator_0]
  properties {pitchCurve {editorCurveData {keys [{time=0}]}}, volumeCurve {editorCurveData {keys [<at 0 0>,<at 1 1>]}}}
 }
 {id=Modulator_0, children [QualityLevel_0], properties {pitchMax=1.2}}
 {id=QualityLevel_0, children [WavePlayer_12, WavePlayer_12, null, null]}
 {
  id=WavePlayer_12
  properties {bLooping=true, soundWaveAssetPtr=/Game/Sounds/Artemis/Environment_Actors/HoverPlatform/Spot_HoverPlatform_Thruster_Rip_Distant}
 }
]
properties {
 attenuationSettings=/Game/Sounds/Artemis/Environment_Actors/HoverPlatform/Spot_HoverPlatform_Distant_Thruster_Ripper_VSmall_Att
 bOverrideConcurrency=true
 maxDistance=3024
 soundClassObject=&Fort_Ambient_Outdoors_Athena
 volumeMultiplier=0.351
 concurrencyOverrides {maxCount=2, voiceStealReleaseTime=2, volumeScaleMode=Distance}
}
