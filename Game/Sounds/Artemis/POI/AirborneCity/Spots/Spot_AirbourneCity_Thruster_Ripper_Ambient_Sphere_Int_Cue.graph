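# Sound cue for the interior ambient spot loop of an Airborne City thruster "ripper", using a distant hover-platform thruster recording.
# A quality-level node feeds one looping wave player on its first two tiers, leaving the other two empty so the sound drops out there.
# It plays quietly in the indoor ambient sound class with a tiny distant-thruster attenuation.
# A concurrency override caps simultaneous instances at a few, scaling volume by distance.
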SoundCue /Game/Sounds/Artemis/POI/AirborneCity/Spots/Spot_AirbourneCity_Thruster_Ripper_Ambient_Sphere_Int_Cue
root=QualityLevel_0
nodes [
 {id=QualityLevel_0, children [WavePlayer_12, WavePlayer_12, null, null]}
 {
  id=WavePlayer_12
  properties {bLooping=true, soundWaveAssetPtr=/Game/Sounds/Artemis/Environment_Actors/HoverPlatform/Spot_HoverPlatform_Thruster_Rip_Distant}
 }
]
properties {
 attenuationSettings=/Game/Sounds/Artemis/Environment_Actors/HoverPlatform/Spot_HoverPlatform_Distant_Thruster_Ripper_Tiny_Att
 bOverrideConcurrency=true
 maxDistance=1250
 soundClassObject=&Fort_Ambient_Indoors_Athena
 volumeMultiplier=0.251
 concurrencyOverrides {maxCount=3, volumeScaleMode=Distance}
}
