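# Looping ambient spot cue of creaking, groaning metal, apparently for a balancing car at the AirborneCity POI.
# Each loop waits a random delay, then layers a random reused metal stress groan at half volume with a random Big Bridge creak, both pitched up with slight variation.
# Capsule attenuation adds distance filtering and listener focus, and concurrency is capped at two.
# Only the first two of four quality-level slots are wired, leaving the others silent.
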SoundCue /Game/Sounds/Artemis/POI/AirborneCity/Spots/Spot_Metal_Car_Balancing_Creak_Loop_Cue
root=QualityLevel_0
nodes [
 {id=Delay_0, children [Modulator_2], properties {delayMax=7, delayMin=3}}
 {id=Looping_0, children [Delay_0]}
 {id=Mixer_1, children [Modulator_0,Modulator_3], properties {inputVolume [0.5,1]}}
 {id=Modulator_0, children [Random_0], properties {pitchMax=1.25, pitchMin=1.2}}
 {id=Modulator_2, children [Mixer_1], properties {volumeMin=0.8}}
 {id=Modulator_3, children [Random_1], properties {pitchMax=1.25, pitchMin=1.2}}
 {id=QualityLevel_0, children [Looping_0,Looping_0,null,null]}
 {
  id=Random_0
  children [WavePlayer_1, WavePlayer_2, WavePlayer_3, WavePlayer_4, WavePlayer_5, WavePlayer_13, WavePlayer_0, WavePlayer_6, WavePlayer_7]
  properties {weights [1,1,1,1,1,1,1,1,1]}
 }
 {
  id=Random_1
  children [WavePlayer_8, WavePlayer_9, WavePlayer_10, WavePlayer_11, WavePlayer_12, WavePlayer_17, WavePlayer_18, WavePlayer_19]
  properties {weights [1,1,1,1,1,1,1,1]}
 }
 {id=WavePlayer_0, properties {soundWaveAssetPtr=/Game/Sounds/Apollo/Biomes/Generic/AMB_Metal_Stress_Wronk_Groan_Knock_09}}
 {id=WavePlayer_1, properties {soundWaveAssetPtr=/Game/Sounds/Apollo/Biomes/Generic/AMB_Metal_Stress_Wronk_Groan_Knock_01}}
 {id=WavePlayer_10, properties {soundWaveAssetPtr=/Game/Sounds/Artemis/POI/Big_Bridge/Spots/Spot_Big_Bridge_Metal_Creak_02}}
 {id=WavePlayer_11, properties {soundWaveAssetPtr=/Game/Sounds/Artemis/POI/Big_Bridge/Spots/Spot_Big_Bridge_Metal_Creak_03}}
 {id=WavePlayer_12, properties {soundWaveAssetPtr=/Game/Sounds/Artemis/POI/Big_Bridge/Spots/Spot_Big_Bridge_Metal_Creak_04}}
 {id=WavePlayer_13, properties {soundWaveAssetPtr=/Game/Sounds/Apollo/Biomes/Generic/AMB_Metal_Stress_Wronk_Groan_Knock_02}}
 {id=WavePlayer_17, properties {soundWaveAssetPtr=/Game/Sounds/Artemis/POI/Big_Bridge/Spots/Spot_Big_Bridge_Metal_Creak_05}}
 {id=WavePlayer_18, properties {soundWaveAssetPtr=/Game/Sounds/Artemis/POI/Big_Bridge/Spots/Spot_Big_Bridge_Metal_Creak_Large_01}}
 {id=WavePlayer_19, properties {soundWaveAssetPtr=/Game/Sounds/Artemis/POI/Big_Bridge/Spots/Spot_Big_Bridge_Metal_Creak_Large_02}}
 {id=WavePlayer_2, properties {soundWaveAssetPtr=/Game/Sounds/Apollo/Biomes/Generic/AMB_Metal_Stress_Wronk_Groan_Knock_06}}
 {id=WavePlayer_3, properties {soundWaveAssetPtr=/Game/Sounds/Apollo/Biomes/Generic/AMB_Metal_Stress_Wronk_Groan_Knock_05}}
 {id=WavePlayer_4, properties {soundWaveAssetPtr=/Game/Sounds/Apollo/Biomes/Generic/AMB_Metal_Stress_Wronk_Groan_Knock_04}}
 {id=WavePlayer_5, properties {soundWaveAssetPtr=/Game/Sounds/Apollo/Biomes/Generic/AMB_Metal_Stress_Wronk_Groan_Knock_03}}
 {id=WavePlayer_6, properties {soundWaveAssetPtr=/Game/Sounds/Apollo/Biomes/Generic/AMB_Metal_Stress_Wronk_Groan_Knock_07}}
 {id=WavePlayer_7, properties {soundWaveAssetPtr=/Game/Sounds/Apollo/Biomes/Generic/AMB_Metal_Stress_Wronk_Groan_Knock_08}}
 {id=WavePlayer_8, properties {soundWaveAssetPtr=/Game/Sounds/Artemis/POI/Big_Bridge/Spots/Spot_Big_Bridge_Metal_Creak_Large_03}}
 {id=WavePlayer_9, properties {soundWaveAssetPtr=/Game/Sounds/Artemis/POI/Big_Bridge/Spots/Spot_Big_Bridge_Metal_Creak_01}}
]
properties {
 bEnableSubmixSends=false
 bOverrideAttenuation=true
 bOverrideConcurrency=true
 maxDistance=1300
 soundClassObject=&Fort_Ambient_Props_Athena
 volumeMultiplier=0.5
 attenuationOverrides {
  attenuationShape=Capsule
  bAttenuateWithLPF=true
  bEnableFocusInterpolation=true
  bEnableListenerFocus=true
  falloffDistance=1000
  focusAzimuth=25
  focusReleaseInterpSpeed=0.4
  hPFFrequencyAtMax=350
  lPFFrequencyAtMax=1000
  lPFRadiusMax=1250
  lPFRadiusMin=750
  nonFocusAzimuth=55
  nonFocusVolumeAttenuation=0.4
  stereoSpread=100
  attenuationShapeExtents=<300 300>
 }
 concurrencyOverrides {maxCount=2, voiceStealReleaseTime=5}
 preEffectBusSends [{}]
}
