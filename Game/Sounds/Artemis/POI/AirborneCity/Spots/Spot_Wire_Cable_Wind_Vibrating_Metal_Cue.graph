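# Looping spot ambience cue of wind-vibrated metal cable for the AirborneCity POI, reusing Blimps_Crashsites waves and attenuation.
# A Random node picks a steady or gusty loop, which gets random pitch modulation, then looping envelopers for a slow pitch drift and longer volume swells.
# A short fade-in envelope and slight start delay wrap the chain; a QualityLevel node leaves its last two slots silent.
# A concurrency override caps it at two voices.
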SoundCue /Game/Sounds/Artemis/POI/AirborneCity/Spots/Spot_Wire_Cable_Wind_Vibrating_Metal_Cue
root=Delay_0
nodes [
 {id=Delay_0, children [QualityLevel_0], properties {delayMax=0.1, delayMin=0.05}}
 {
  id=Enveloper_0
  children [Modulator_1]
  properties {
   bLoop=true
   bLoopIndefinitely=true
   loopEnd=30
   pitchCurve {editorCurveData {keys [{interpMode=RCIM_Cubic, time=0}, <at 15.9468 0.9 RCIM_Cubic>, {interpMode=RCIM_Cubic, time=30}]}}
   volumeCurve {editorCurveData {keys [<at 0 1>]}}
  }
 }
 {
  id=Enveloper_29
  children [Enveloper_3]
  properties {
   pitchCurve {editorCurveData {keys [{time=0}]}}
   volumeCurve {editorCurveData {keys [{arriveTangent=3.4293, interpMode=RCIM_Cubic, leaveTangent=3.4293, tangentMode=RCTM_User, time=0, value=0}, <at 0.4 1 RCIM_Cubic>]}}
  }
 }
 {
  id=Enveloper_3
  children [Enveloper_0]
  properties {
   bLoop=true
   bLoopIndefinitely=true
   loopEnd=120
   pitchCurve {editorCurveData {keys [{time=0}]}}
   volumeCurve {editorCurveData {keys [
      {arriveTangent=0.0001, interpMode=RCIM_Cubic, leaveTangent=0.0082, tangentMode=RCTM_User, time=0.5, value=0.3803}
      {arriveTangent=0.0318, interpMode=RCIM_Cubic, leaveTangent=0.0318, time=10, value=0.6}
      {arriveTangent=-0.0028, interpMode=RCIM_Cubic, leaveTangent=-0.0028, time=20, value=1}
      <at 45.8196 0.5 RCIM_Cubic>
      {arriveTangent=0.0151, interpMode=RCIM_Cubic, leaveTangent=0.0151, time=60, value=1}
      {arriveTangent=-0.0202, interpMode=RCIM_Cubic, leaveTangent=-0.0202, tangentMode=RCTM_User, time=82.1697, value=1.05}
      <at 120 0.5>
     ]}}
  }
 }
 {id=Modulator_1, children [Random_0], properties {pitchMax=1.2, pitchMin=1}}
 {id=QualityLevel_0, children [Enveloper_29,Enveloper_29,null,null]}
 {id=Random_0, children [WavePlayer_1,WavePlayer_0], properties {weights [1,1]}}
 {
  id=WavePlayer_0
  properties {bLooping=true, soundWaveAssetPtr=/Game/Sounds/Artemis/POI/Blimps_Crashsites/Spots/Spot_Wind_Excited_Metal_Vibrating_Gusty_Loop}
 }
 {
  id=WavePlayer_1
  properties {bLooping=true, soundWaveAssetPtr=/Game/Sounds/Artemis/POI/Blimps_Crashsites/Spots/Spot_Wind_Excited_Metal_Vibrating_Loop}
 }
]
properties {
 attenuationSettings=/Game/Sounds/Artemis/POI/Blimps_Crashsites/Spots/Spot_Wind_Metal_Vibrating_Large_Focused_Linear_Att
 bEnableBusSends=false
 bEnableSubmixSends=false
 bOverrideConcurrency=true
 maxDistance=1865.4736
 soundClassObject=&Fort_Gadgets
 volumeMultiplier=0.6
 concurrencyOverrides {maxCount=2, voiceStealReleaseTime=2}
}
